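# Translucent UI material in the victory HUD folder that builds a triangle-patterned radial dissolve mask.
# Screen-aligned UVs, scaled by PatternAmount and shifted by PatternOffset, sample the triangle texture as the SmoothStep alpha.
# An inverted exponential radial gradient sets the threshold, with EdgeSmooth added for the upper bound.
# PatternOffset and RadiusLocation use only red for both axes; no connection to material outputs is shown.
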Material /Game/Athena/HUD/Materials/Victory/M_UI_Dissolve
graph {
 properties {blendMode=BLEND_Translucent, materialDomain=MD_UI, pixelDepthOffsetMode=PDOM_Legacy, usedShadingModels=Default Lit}
 nodes [
  Constant() {r=1}
  n1 = Multiply(
   A=MaterialFunctionCall() {materialFunction=&ScreenAlignedUVs}."X 100%, Y 100%"
   B=ScalarParameter() {defaultValue=20, parameterName=PatternAmount}
  )
  n2 = VectorParameter() {parameterName=PatternOffset, defaultValue=<rgba 0 0 0 1>}
  n3 = AppendVector(A=@n2.R, B=@n2.R)
  n4 = VectorParameter() {parameterName=RadiusLocation, defaultValue=<rgba 1 0.5 0 1>}
  n5 = OneMinus(None=MaterialFunctionCall(
    CenterPosition=AppendVector(A=@n4.R, B=@n4.R)
    Radius=ScalarParameter() {defaultValue=6, parameterName=Radius}
    Density=ScalarParameter() {defaultValue=2, parameterName=Density}
   ) {materialFunction=&RadialGradientExponential}.RadialGradientExponential)
  MaterialFunctionCall(
   Alpha=TextureSample(UVs=Add(A=@n1, B=@n3)) {texture=/Game/Global/Textures/Default/Masks_and_Gradients/T_TrianglePattern}.RGB
   Min=@n5
   Max=Add(A=@n5, B=ScalarParameter() {parameterName=EdgeSmooth})
  ) {materialFunction=&SmoothStep}
 ]
}
