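# Material function that produces a positionable, angled shine from the T-UI-Shine texture, filed with the Athena victory HUD materials.
# CustomRotator turns the UVs by the Rotation input around RotationCenter, then ShineOffset shifts the rotated result along U only.
# The shifted UVs sample the texture through the shared clamp sampler, so it does not tile, and the function outputs RGB.
# All inputs have preview defaults; the UVs input is typed scalar despite its TextureCoordinate default.
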MaterialFunction /Game/Athena/HUD/Materials/Victory/MF_Shine
graph {nodes [
  n1 = MaterialFunctionCall(
   UVs=FunctionInput(Preview=TextureCoordinate()) {bUsePreviewValueAsDefault=true, inputName=UVs, inputType=FunctionInput_Scalar}
   Rotation Center=FunctionInput(Preview=Constant2Vector()) {bUsePreviewValueAsDefault=true, inputName=RotationCenter, inputType=FunctionInput_Vector2}
   "Rotation Angle (0-1)"=FunctionInput(Preview=Constant() {r=0.9}) {bUsePreviewValueAsDefault=true, inputName=Rotation, inputType=FunctionInput_Scalar}
  ) {materialFunction=&CustomRotator}
  n2 = AppendVector(
   A=FunctionInput(Preview=Constant() {r=-0.5453}) {bUsePreviewValueAsDefault=true, inputName=ShineOffset, inputType=FunctionInput_Scalar}
   B=Constant()
  )
  FunctionOutput(None=TextureSample(UVs=Add(A=@n1.Rotated Values, B=@n2)) {samplerSource=SSM_Clamp_WorldGroupSettings, texture=/Game/UI/Foundation/Textures/Pattern/T-UI-Shine}.RGB)
 ]}
