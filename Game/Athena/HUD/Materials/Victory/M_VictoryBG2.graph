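# Translucent UI material for a victory screen background plate with animatable bolt, spike, shine and reveal parameters.
# Color layers a rotatable two-color gradient, radial highlight, bolt shape cut by three rotated smoothstep edges, and two transformable spike samples from the plate texture's green channel, finished with a color-dodge shine.
# A mask multiplies the antialias-thresholded red-channel plate silhouette by a reveal that thresholds a screen-aligned triangle pattern against a radial gradient.
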Material /Game/Athena/HUD/Materials/Victory/M_VictoryBG2
graph {
 properties {blendMode=BLEND_Translucent, materialDomain=MD_UI, pixelDepthOffsetMode=PDOM_Legacy, usedShadingModels=Default Lit}
 nodes [
  n1 = VectorParameter() {group=Shine, parameterName=ShineColor, defaultValue=<rgba 1 1 1 1>}
  n2 = MaterialFunctionCall!(
   ShineOffset=ScalarParameter() {defaultValue=-1, group=Shine, parameterName=ShineOffset, sliderMax=2, sliderMin=-2}
   Rotation=ScalarParameter() {defaultValue=0.0476, group=Shine, parameterName=Rotation, sliderMax=1}
   RotationCenter
   UVs
  ) {materialFunction=&/MF_Shine}
  n3 = TextureCoordinate()
  n4 = ScalarParameter() {group=BGColors, parameterName=PlateGradientRotation}
  n5 = LinearInterpolate(
   A=VectorParameter() {group=BGColors, parameterName=BGColor1, defaultValue=<rgba 0 0.107 0.5395 1>}.RGB
   B=VectorParameter() {group=BGColors, parameterName=BGColor2, defaultValue=<rgba 0 0.107 0.5395 1>}.RGB
   Alpha=ComponentMask(None=MaterialFunctionCall(UVs=@n3, "Rotation Angle (0-1)"=@n4) {materialFunction=&CustomRotator}.Rotated Values) {g=true}
  )
  n6 = VectorParameter() {group=BGColors, parameterName=HighlightColor, defaultValue=<rgba 0 0.7529 1 1>}
  n7 = VectorParameter() {group=BGHighlight, parameterName=HighlightPosition, defaultValue=<rgba 0 1 0 1>}
  n8 = MaterialFunctionCall(
   CenterPosition=AppendVector(A=@n7.R, B=@n7.R)
   Radius=ScalarParameter() {defaultValue=1, group=BGHighlight, parameterName=HighlightRadius}
   Density=ScalarParameter() {defaultValue=1, group=BGHighlight, parameterName=HighlightDensity}
  ) {materialFunction=&RadialGradientExponential}
  n9 = MaterialFunctionCall() {materialFunction=&GetUserInterfaceUV}
  n10 = ComponentMask(None=MaterialFunctionCall(UVs=@n9.Normalized UV, "Rotation Angle (0-1)"=ScalarParameter() {defaultValue=0.25, group=BoltPlate, parameterName=Angle1Rotation}) {materialFunction=&CustomRotator}.Rotated Values) {g=true}
  n11 = ScalarParameter() {defaultValue=0.5, group=BoltPlate, parameterName=Angle1Edge}
  n12 = Add(A=@n11, B=ScalarParameter() {defaultValue=0.01, group=BoltPlate, parameterName=AngleEdgeSoftness})
  n13 = ScalarParameter() {defaultValue=0.25, group=BoltPlate, parameterName=Angle2Rotation}
  n14 = ScalarParameter() {defaultValue=0.5, group=BoltPlate, parameterName=Angle2Edge}
  n15 = MaterialFunctionCall(
   Alpha=ComponentMask(None=MaterialFunctionCall(UVs=@n9.Normalized UV, "Rotation Angle (0-1)"=@n13) {materialFunction=&CustomRotator}.Rotated Values) {g=true}
   Min=@n14
   Max=Add(A=@n14, B=ScalarParameter() {defaultValue=0.01, group=BoltPlate, parameterName=AngleEdgeSoftness})
  ) {materialFunction=&SmoothStep}
  n16 = ScalarParameter() {defaultValue=0.2, group=BoltPlate, parameterName=Angle3Rotation}
  n17 = ScalarParameter() {defaultValue=0.5, group=BoltPlate, parameterName=Angle3Edge}
  n18 = MaterialFunctionCall(
   Alpha=ComponentMask(None=MaterialFunctionCall(UVs=@n9.Normalized UV, "Rotation Angle (0-1)"=@n16) {materialFunction=&CustomRotator}.Rotated Values) {g=true}
   Min=@n17
   Max=Add(A=@n17, B=ScalarParameter() {defaultValue=0.01, group=BoltPlate, parameterName=AngleEdgeSoftness})
  ) {materialFunction=&SmoothStep}
  n19 = Clamp(None=Subtract(
    A=MaterialFunctionCall(Alpha=@n10, Min=@n11, Max=@n12) {materialFunction=&SmoothStep}.Result
    B=Subtract(A=@n15.Result, B=@n18.Result)
   ))
  n20 = ScalarParameter() {group=BoltPlate, parameterName=BoltPlateAmount}
  n21 = VectorParameter() {group=Spikes, parameterName=SpikeColor, defaultValue=<rgba 0.1981 0.6445 1 1>}
  n22 = ScalarParameter() {defaultValue=0.2, group=Spikes, parameterName=SpikeThreshold}
  n23 = VectorParameter() {group=Spike1, parameterName=Spike1Scale, defaultValue=<rgba 0 0 1 1>}
  n24 = AppendVector(A=@n23.B, B=@n23.B)
  n25 = Constant2Vector() {g=0.5, r=0.8}
  n26 = MaterialFunctionCall(UVs=TextureCoordinate(), Rotation Center=@n25, "Rotation Angle (0-1)"=ScalarParameter() {group=Spike1, parameterName=Spike1Rotation}) {materialFunction=&CustomRotator}
  n27 = Constant() {r=512}
  n28 = AppendVector(A=AppendVector(A=@n23.R, B=@n23.R), B=@n27)
  n29 = VectorParameter() {group=Spike1, parameterName=Spike1Offset, defaultValue=<rgba 0 0 0 1>}
  n30 = TextureSampleParameter2D(UVs=Add(
    A=MaterialFunctionCall!(Texture Scale=@n24, UVs=@n26.Rotated Values, Offset XY Texturesize Z=@n28) {materialFunction=/Engine/Functions/Engine_MaterialFunctions01/Texturing/ScaleUVsAroundPoint}.UVs
    B=AppendVector(A=@n29.R, B=@n29.R)
   )) {parameterName=PlateTexture, samplerSource=SSM_Clamp_WorldGroupSettings, texture=&/T_VictoryBG2}
  n31 = ScalarParameter() {group=Spikes, parameterName=SpikeOpacity, sliderMax=0.6}
  n32 = LinearInterpolate(
   A=LinearInterpolate(
    A=LinearInterpolate(A=@n5, B=@n6.RGB, Alpha=@n8.RadialGradientExponential)
    B=VectorParameter() {group=BoltPlate, parameterName=BoltPlateColor, defaultValue=<rgba 0 0.016 0.3385 1>}.RGB
    Alpha=Multiply(A=@n19, B=@n20)
   )
   B=@n21.RGB
   Alpha=Multiply(
    A=MaterialFunctionCall!(Threshold=@n22, Gradient=@n30.G, Filter Width Override, Filter Width Fudge Factor) {materialFunction=/Game/UI/Materials/AAThreshold}.Result
    B=@n31
   )
  )
  n33 = VectorParameter() {group=Spike2, parameterName=Spike2Scale, defaultValue=<rgba 0 0 1 1>}
  n34 = AppendVector(A=@n33.B, B=@n33.B)
  n35 = MaterialFunctionCall(UVs=TextureCoordinate(), Rotation Center=@n25, "Rotation Angle (0-1)"=ScalarParameter() {group=Spike2, parameterName=Spike2Rotation}) {materialFunction=&CustomRotator}
  n36 = AppendVector(A=AppendVector(A=@n33.R, B=@n33.R), B=@n27)
  n37 = VectorParameter() {group=Spike2, parameterName=Spike2Offset, defaultValue=<rgba 0 0 0 1>}
  n38 = TextureSampleParameter2D(UVs=Add(
    A=MaterialFunctionCall(Texture Scale=@n34, UVs=@n35.Rotated Values, Offset XY Texturesize Z=@n36) {materialFunction=&ScaleUVsAroundPoint}.UVs
    B=AppendVector(A=@n37.R, B=@n37.R)
   )) {parameterName=PlateTexture, samplerSource=SSM_Clamp_WorldGroupSettings, texture=&T_VictoryBG2}
  n39 = Multiply(
   A=MaterialFunctionCall(Threshold=ScalarParameter() {defaultValue=0.2, group=Spikes, parameterName=SpikeThreshold}, Gradient=@n38.G) {materialFunction=&AAThreshold}.Result
   B=@n31
  )
  Clamp(None=MaterialFunctionCall!(Blend=Multiply(A=@n1.A, B=@n2.Result), Base=LinearInterpolate(A=@n32, B=@n21.RGB, Alpha=@n39)) {materialFunction=/Engine/Functions/Engine_MaterialFunctions03/Blends/Blend_ColorDodge}.Result)
  MaterialFunctionCall() {materialFunction=&GetUserInterfaceUV}
  n40 = VectorParameter() {parameterName=PlateScale, defaultValue=<rgba 1 1 0 1>}
  n41 = TextureSampleParameter2D(UVs=MaterialFunctionCall(Texture Scale=AppendVector(A=@n40.R, B=@n40.R)) {materialFunction=&ScaleUVsByCenter}.UVs) {parameterName=PlateTexture, samplerSource=SSM_Clamp_WorldGroupSettings, texture=&T_VictoryBG2}
  n42 = MaterialFunctionCall(
   Radius=ScalarParameter() {defaultValue=43.4849, group=RevealMask, parameterName=RevealRadius}
   Density=ScalarParameter() {defaultValue=8.9055, group=RevealMask, parameterName=RevealDensity}
  ) {materialFunction=&RadialGradientExponential}
  n43 = Constant2Vector() {g=256, r=256}
  n44 = TextureSampleParameter2D(UVs=Multiply(
    A=MaterialFunctionCall!(TextureResolution=@n43) {materialFunction=/Engine/Functions/Engine_MaterialFunctions02/Texturing/ScreenAlignedPixelToPixelUVs}.Result
    B=ScalarParameter() {defaultValue=1, group=RevealMask, parameterName=PatternRep}
   )) {group=RevealMask, parameterName=PatternTexture, texture=/Game/Global/Textures/Default/Masks_and_Gradients/T_TrianglePattern}
  Multiply(
   A=MaterialFunctionCall(Threshold=ScalarParameter() {defaultValue=0.2, parameterName=BGThreshold}, Gradient=@n41.R) {materialFunction=&AAThreshold}.Result
   B=OneMinus(None=MaterialFunctionCall(Threshold=@n42.RadialGradientExponential, Gradient=@n44.R) {materialFunction=&AAThreshold}.Result)
  )
  TextureCoordinate()
  TextureCoordinate()
  TextureCoordinate()
 ]
}
