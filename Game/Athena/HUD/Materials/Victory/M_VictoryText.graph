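# UI material for victory text that shades it with a gradient and a shine highlight.
# An offset, rotatable UV coordinate, raised to a power, samples a curve-atlas gradient, which is lerped with FillColor using that color as alpha.
# MF_Shine, positioned by a static switch between time-driven looping and a manual offset, is color-dodged over the gradient, then clamped.
# Another static switch selects 9-slice-aware UVs.
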Material /Game/Athena/HUD/Materials/Victory/M_VictoryText
graph {
 properties {materialDomain=MD_UI, pixelDepthOffsetMode=PDOM_Legacy, usedShadingModels=Default Lit}
 nodes [
  n1 = VectorParameter() {group=Shine, parameterName=ShineColor, defaultValue=<rgba 1 1 1 1>}
  n2 = VectorParameter() {group=Shine, parameterName=ShineExtents, defaultValue=<rgba -1 1 0 1>}
  n3 = Divide(A=Fmod(A=Time(), B=Constant() {r=1}), B=ScalarParameter() {defaultValue=1, group=Shine, parameterName=ShineSpeed})
  n4 = MaterialFunctionCall!(
   ShineOffset=StaticSwitchParameter(
    True=LinearInterpolate(A=@n2.R, B=@n2.R, Alpha=@n3)
    False=ScalarParameter() {defaultValue=-1.001, group=Shine, parameterName=ShineOffset, sliderMax=2, sliderMin=-2}
   ) {group=Shine, parameterName=autoShine}
   Rotation=ScalarParameter() {defaultValue=0.0476, group=Shine, parameterName=ShineRotation, sliderMax=1}
   RotationCenter
   UVs=StaticSwitchParameter(True=MaterialFunctionCall() {materialFunction=&GetUserInterfaceUV}.Normalized UV, False=TextureCoordinate()) {parameterName=is9slice}
  ) {materialFunction=&/MF_Shine}
  n5 = VectorParameter() {parameterName=FillColor, defaultValue=<rgba 1 1 1 1>}
  n6 = MaterialFunctionCall() {materialFunction=&GetUserInterfaceUV}
  n7 = TextureCoordinate()
  n8 = VectorParameter() {parameterName=GradientOffset, defaultValue=<rgba 0 0 0 1>}
  n9 = VectorParameter() {parameterName=RotationCenter, defaultValue=<rgba 0.5 0.5 0 1>}
  n10 = MaterialFunctionCall(
   UVs=Add(A=StaticSwitchParameter(True=@n6."9-Slice UV (No Tiling)", False=@n7) {parameterName=is9slice}, B=AppendVector(A=@n8.R, B=@n8.R))
   Rotation Center=AppendVector(A=@n9.R, B=@n9.R)
   "Rotation Angle (0-1)"=ScalarParameter() {parameterName=Rotation}
  ) {materialFunction=&CustomRotator}
  n11 = ScalarParameter() {defaultValue=1, parameterName=RangePower}
  n12 = CurveAtlasRowParameter(CurveTime=Power(Base=ComponentMask(None=@n10.Rotated Values) {g=true}, Exp=Add(A=@n11) {constB=0})) {
   atlas=&/NewCurveLinearColorAtlas
   curve=&/VictoryTextGradientCurve
   parameterName=CurveParams
  }
  Clamp(None=MaterialFunctionCall!(Blend=Multiply(A=@n1.A, B=@n4.Result), Base=LinearInterpolate(A=@n5.RGB, B=@n12, Alpha=@n5.RGB)) {materialFunction=/Engine/Functions/Engine_MaterialFunctions03/Blends/Blend_ColorDodge}.Result)
 ]
}
