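# Additive UI material that draws light shafts for the Athena HUD victory display.
# The red channel of the shaft texture is multiplied by a SmoothStep of itself, bounded by ShaftMin and ShaftMax, to shape the shafts.
# The result is masked by an exponential radial gradient with parameterized radius and density.
# RadiusOffset's red value shifts the gradient's UVs equally on both axes, moving its center diagonally.
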Material /Game/Athena/HUD/Materials/Victory/M_Victory_LightShafts
graph {
 properties {blendMode=BLEND_Additive, materialDomain=MD_UI, pixelDepthOffsetMode=PDOM_Legacy, usedShadingModels=Default Lit}
 nodes [
  Constant3Vector() {constant=<rgba 1 1 1 0>}
  n1 = TextureCoordinate()
  n2 = VectorParameter() {parameterName=RadiusOffset, defaultValue=<rgba 0 0 0 1>}
  n3 = AppendVector(A=@n2.R, B=@n2.R)
  n4 = TextureSample() {samplerType=SAMPLERTYPE_LinearColor, texture=&/T_Victory_LightShafts}
  n5 = ScalarParameter() {parameterName=ShaftMin}
  n6 = ScalarParameter() {defaultValue=1, parameterName=ShaftMax}
  Multiply(
   A=MaterialFunctionCall(
    UVs=Add(A=@n1, B=@n3)
    Radius=ScalarParameter() {defaultValue=0.5, parameterName="Radius (S)"}
    Density=ScalarParameter() {defaultValue=2.3, parameterName="Density (S)"}
   ) {materialFunction=&RadialGradientExponential}.RadialGradientExponential
   B=Multiply(
    A=TextureSample() {samplerType=SAMPLERTYPE_LinearColor, texture=&T_Victory_LightShafts}.R
    B=MaterialFunctionCall(Alpha=@n4.R, Min=@n5, Max=@n6) {materialFunction=&SmoothStep}.Result
   )
  )
 ]
}
