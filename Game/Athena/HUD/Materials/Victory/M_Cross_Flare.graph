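# M_Cross_Flare is an additive UI material in the Athena HUD's Victory folder, evidently for a cross-shaped flare glow on the victory display.
# Its only node samples the grayscale T_Cross_Flare texture from the shared Fort effects flares, with no parameters, tinting or animation.
# The graph shows no connection from that sample to a material output.
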Material /Game/Athena/HUD/Materials/Victory/M_Cross_Flare
graph {
 properties {blendMode=BLEND_Additive, materialDomain=MD_UI, pixelDepthOffsetMode=PDOM_Legacy, usedShadingModels=Default Lit}
 nodes [TextureSample() {samplerType=SAMPLERTYPE_Grayscale, texture=/Game/Effects/Fort_Effects/Textures/Flares/T_Cross_Flare}]
}
