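# Additive UI material for the spikeball flare texture, stored among the Athena victory HUD materials.
# A static switch optionally spins the clamped texture's UVs through a Rotator driven by time and RotationSpeed.
# A second switch multiplies the texture color by a FlickerRange lerp driven by a time-based LinearSine.
# Both lerp inputs read the red channel of FlickerRange, so the flicker is a constant multiplier.
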Material /Game/Athena/HUD/Materials/Victory/M_Flare_Spikeball
graph {
 properties {blendMode=BLEND_Additive, materialDomain=MD_UI, pixelDepthOffsetMode=PDOM_Legacy, usedShadingModels=Default Lit}
 nodes [
  n1 = VectorParameter() {group=Flicker, parameterName=FlickerRange, defaultValue=<rgba 1 0.75 0 1>}
  n2 = Time()
  n3 = Constant() {r=1}
  n4 = MaterialFunctionCall(Value=Multiply(A=Fmod(A=@n2, B=@n3), B=ScalarParameter() {defaultValue=1, group=Flicker, parameterName=FlickerSpeed})) {materialFunction=&LinearSine}
  n5 = Multiply(A=Time(), B=ScalarParameter() {defaultValue=2, group=Rotation, parameterName=RotationSpeed})
  n6 = TextureSampleParameter2D(UVs=StaticSwitchParameter(True=Rotator(Time=@n5), False=TextureCoordinate()) {group=Rotation, parameterName=Rotation}) {parameterName=Texture, samplerSource=SSM_Clamp_WorldGroupSettings, texture=/Game/Effects/Fort_Effects/Textures/Flares/T_Flare_Spikeball}
  StaticSwitchParameter(True=Multiply(A=LinearInterpolate(A=@n1.R, B=@n1.R, Alpha=@n4.Linear Sine), B=@n6.RGB), False=@n6.RGB) {group=Flicker, parameterName=Flicker}
 ]
}
