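# Additive UI material for the victory HUD that draws a light flare texture with an optional radial reveal mask.
# The static Reveal switch multiplies the texture by an exponential radial gradient with offset UVs.
# Looping time, from ManualTime or Time×Speed when that is zero, drives the offset lerps.
# As wired, both lerp ends read Extents minimums, so the offset stays fixed, and gradient Density reuses Radius.
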Material /Game/Athena/HUD/Materials/Victory/M_UI_LightFlare
graph {
 properties {blendMode=BLEND_Additive, materialDomain=MD_UI, pixelDepthOffsetMode=PDOM_Legacy, usedShadingModels=Default Lit}
 nodes [
  n1 = VectorParameter() {group=Reveal, parameterName=Extents, channelNames {a=yMax, b=yMin, g=xMax, r=xMin}, defaultValue=<rgba -1 1 0 1>}
  n2 = ScalarParameter() {group=Reveal, parameterName=ManualTime}
  n3 = Time()
  n4 = ScalarParameter() {defaultValue=1, group=Reveal, parameterName=Speed}
  n5 = Fmod(A=If(A=@n2, B=Constant(), "A > B"=@n2, "A == B"=Multiply(A=@n3, B=@n4), "A < B"=@n2), B=Constant() {r=1})
  n6 = Add(
   A=TextureCoordinate()
   B=AppendVector(
    A=LinearInterpolate(A=@n1.xMin, B=@n1.xMin, Alpha=@n5) {constA=1, constB=-1}
    B=LinearInterpolate(A=@n1.yMin, B=@n1.yMin, Alpha=@n5) {constA=1, constB=-1}
   )
  )
  n7 = VectorParameter() {group=Reveal, parameterName=Gradient, channelNames {a=Y, b=X, g=Density, r=Radius}, defaultValue=<rgba 0.5 2 0.5 0.5>}
  n8 = AppendVector(A=@n7.X, B=@n7.X)
  n9 = TextureSampleParameter2D() {parameterName=Texture, texture=&/T_UI_LightFlare}
  StaticSwitchParameter(
   True=Multiply(
    A=MaterialFunctionCall(UVs=@n6, CenterPosition=@n8, Radius=@n7.Radius, Density=@n7.Radius) {materialFunction=&RadialGradientExponential}.RadialGradientExponential
    B=@n9.RGB
   )
   False=@n9.RGB
  ) {group=Reveal, parameterName=Reveal}
 ]
}
